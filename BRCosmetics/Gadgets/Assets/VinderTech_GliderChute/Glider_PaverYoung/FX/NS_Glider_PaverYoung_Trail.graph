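# Niagara trail effect for the PaverYoung glider: thruster flames, inner flames, glows with lights, wing sprites and speed lines.
# Thruster and glow particles are burst-spawned, partitioned onto skeletal mesh sockets and split by spawn group into mainly orange and cyan looks.
# System age curves drive an ignition flare, the timeline and ForwardDot modulate intensity and flame length, and front-end modifiers retune it for preview.
# The Minimal and SpeedLine emitters are disabled.
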NiagaraSystem /BRCosmetics/Gadgets/Assets/VinderTech_GliderChute/Glider_PaverYoung/FX/NS_Glider_PaverYoung_Trail
dependencies {
 usedDataInterfaces [
  &NiagaraDataInterfaceCurve
  &NiagaraDataInterfaceArrayInt32
  &NiagaraDataInterfaceArrayFloat
  &NiagaraDataInterfaceVector2DCurve
  &NiagaraDataInterfaceColorCurve
  &NiagaraDataInterfaceSkeletalMesh
 ]
 usedDynamicInputs [&Boolean_AndOperation, &SetBoolByFloatComparison]
 usedModules [
  &SystemState
  &self:SystemSpawnScript.SystemScriptSource.SystemScriptGraph.NiagaraNodeAssignment_0.SetVariables_#1
  &Timeline
  &FN_Integrator
  &self:SystemSpawnScript.SystemScriptSource.SystemScriptGraph.NiagaraNodeAssignment_1.SetVariables_#2
 ]
}
emitters {
 Glows {
  emitterUpdateScript [
   &EmitterState() {module=EmitterState001}
   &SpawnBurst_Instantaneous(Spawn Count=2, Use Spawn Probability=false, UseLoopCountLimit=true)
   &SpawnBurst_Instantaneous(Spawn Count=4, Spawn Group=1, Use Spawn Probability=false, UseLoopCountLimit=true) {module=SpawnBurst_Instantaneous001}
  ]
  particleSpawnScript [
   &PartitionParticles(Number Of Partitions=6)
   &InitializeParticle(
    Color=&MaskLinearColor(Mask Source=0, Mask Parameter=?, Mask=0, Masked Linear Color=<rgba 1 0.2833 0 1>, Passthrough Linear Color=<rgba 0 0.8 1 1>)
    Sprite Size Mode=::NewEnumerator3
    Uniform Sprite Size=&FN_FrontEndFloatModifier(Gameplay=60, Front End=45)
    UsePositionOffset=false
    Write Material Random=true
   )
   &CameraOffset(Camera Offset Amount=&MaskFloatBySpawnGroup(Masked Float=15, Passthrough Float=20))
  ]
  particleUpdateScript [
   &ParticleState(Kill Particles When Lifetime Has Elapsed=false, Let Infinitely Lived Particles Die When Emitter Deactivates=true)
   &ScaleSpriteSize(
    Scale Sprite Size Mode=::NewEnumerator0
    Uniform Scale Factor=&Multiply_Float(
     A=&Multiply_Float(
      A=&Lerp_Float(A=1, B=1.5, Alpha=$System.Timeline.NormalizedTime)
      B=&MaskFloatBySpawnGroup(Masked Float=1, Passthrough Float=0.5)
     )
     B=&FloatFromCurve(
      FloatCurve=DataInterface() {propertyValues {Curve {
         defaultValue=1
         keys [
          {interpMode=RCIM_Cubic, tangentMode=RCTM_User, time=0, value=1}
          <at 0.5 1.5 RCIM_Cubic>
          {interpMode=RCIM_Cubic, tangentMode=RCTM_User, time=0.75, value=1}
         ]
        }}}
      CurveIndex=$Engine.System.Age
     )
    )
   )
   &SkeletalMeshLocation(
    Bone / Socket Index=$Particles.PartitionParticles.Partition
    MaskSpawnGroup=false
    Mesh Sampling Type=::NewEnumerator5
    Orientation Sampling=::NewEnumerator1
    Sampled Position Offset Coordinate Space:ENiagaraSkelMeshTransforms=::NewEnumerator2
    Skeletal Mesh=$System.Skeletal Mesh
    Socket Sampling Mode=::NewEnumerator5
   )
   &FN_Light_Attributes(
    Color=$Particles.Color
    Create Color Attribute=true
    Create Exponent Attribute=true
    Create Radius Attribute=true
    Exponent=1
    Radius=&MaskFloatBySpawnGroup(Masked Float=50, Passthrough Float=25)
    Scale Color=&Multiply_Float(
     A=&Lerp_Float(A=3, B=4, Alpha=$System.Timeline.NormalizedTime)
     B=&FloatFromCurve(
      FloatCurve=DataInterface() {propertyValues {Curve {
         defaultValue=1
         keys [{interpMode=RCIM_Cubic, tangentMode=RCTM_User, time=0, value=0}, {interpMode=RCIM_Cubic, tangentMode=RCTM_User, time=1, value=1}]
        }}}
      CurveIndex=$Engine.System.Age
     )
    )
   )
   &DynamicMaterialParameters(
    Index 0 Param 1=3
    Index 0 Param 2=3
    Module.Index0Float0Write=true
    Module.Index0Float1Write=true
    Module.Index0Float2Write=false
    Module.Index0Float3Write=false
   )
   &ScaleColor(
    Scale Alpha=&Multiply_Float(
     A=&Multiply_Float(
      A=&MaskFloat(Mask Source=0, Mask Parameter=?, Mask=0, Masked Float=1, Passthrough Float=0.5)
      B=&Multiply_Float(
       A=&Lerp_Float(A=0.4, B=0.3, Alpha=$System.Timeline.NormalizedTime)
       B=&FloatFromCurve(
        FloatCurve=DataInterface() {propertyValues {Curve {
           defaultValue=1
           keys [
            {interpMode=RCIM_Cubic, tangentMode=RCTM_User, time=0, value=0}
            {arriveTangent=8.5978, interpMode=RCIM_Cubic, leaveTangent=8.5978, tangentMode=RCTM_User, time=0.25, value=1}
            <at 0.35 2 RCIM_Cubic>
            {interpMode=RCIM_Cubic, tangentMode=RCTM_User, time=1, value=1}
           ]
          }}}
        CurveIndex=$Engine.System.Age
       )
      )
     )
     B=&FN_FrontEndFloatModifier(Gameplay=1, Front End=0.25)
    )
    Scale RGB=&VectorFromFloat(Value=&FloatFromCurve(
      FloatCurve=DataInterface() {propertyValues {Curve {
         defaultValue=1
         keys [
          {arriveTangent=4.2857, interpMode=RCIM_Cubic, time=0.25, value=1}
          <at 0.35 2.5 RCIM_Cubic>
          {interpMode=RCIM_Cubic, tangentMode=RCTM_User, time=1, value=1}
         ]
        }}}
      CurveIndex=$Engine.System.Age
     ))
    ScaleA=true
    ScaleRGB=true
   ) {module=ScaleColor002}
   &FN_Sprite_Attributes(
    Color=&MultiplyLinearColor(
     Linear Color=$Particles.Color
     Scale Factor=&MakeVector4FromVectorAndFloat(
      "Vector (RGB)"=&VectorFromFloat(Value=&Lerp_Float(A=1, B=2, Alpha=$System.Timeline.NormalizedTime))
      "Float (Alpha)"=1
     )
    )
    Create Color Attribute=true
    Create Position Attribute=true
    Module Output:ENiagara_AttributeSamplingApplyOutput=::NewEnumerator1
    Position:NiagaraPosition=&AddVectorToPosition(
     Position=$Particles.SkeletalMeshLocation.SampledPosition
     Vector=&MaskVectorBySpawnGroup(
      Masked Vector=&RotateVectorByQuaternion(Vector To Rotate=<0 20 0>, Quaternion=$Particles.SkeletalMeshLocation.SampledRotation)
      Passthrough Vector=&RotateVectorByQuaternion(Vector To Rotate=<10 0 0>, Quaternion=$Particles.SkeletalMeshLocation.SampledRotation)
     )
    )
   )
  ]
  rendererClasses [&NiagaraSpriteRendererProperties, &NiagaraLightRendererProperties]
  renderers [
   {rendererClass=&NiagaraSpriteRendererProperties}
   {rendererClass=&NiagaraLightRendererProperties, rendererIndex=1}
  ]
 }
 InnerFlameMesh {
  emitterUpdateScript [
   &EmitterState()
   &SpawnBurst_Instantaneous(Spawn Count=6, Use Spawn Probability=false, UseLoopCountLimit=true)
   &SpawnBurst_Instantaneous(Spawn Count=12, Spawn Group=1, Use Spawn Probability=false, UseLoopCountLimit=true) {module=SpawnBurst_Instantaneous001}
  ]
  particleSpawnScript [
   &InitializeParticle(
    Color=&MaskLinearColorBySpawnGroup(Masked Linear Color=<rgba 2 0.4559 0 0.25>, Passthrough Linear Color=<rgba 0 1.5 2.5 0.25>)
    Lifetime=2
    Mesh Scale=<0.12 0.65 0.65>
    Mesh Scale Mode=::NewEnumerator4
    UsePositionOffset=false
    Write Material Random=true
   )
   &PartitionParticles(Number Of Partitions=6, Particles Per Partition=3, Partition Mode:ENiagara_PartitionMode_v2=::NewEnumerator1)
   &InitialMeshOrientation(Rotation=<0 0 -0.25>, Use Orientation Vector=true, Use Rotation Vector=true)
   &InitialMeshOrientation(
    Rotation=&MaskVector(
     Mask Source=0
     Mask Parameter=?
     Mask=0
     Masked Vector=&RandomRangeVector(Minimum=<0 0 0>, Maximum=<0 1 0>)
     Passthrough Vector=<0 0 0>
    )
    Use Orientation Vector=true
    Use Rotation Vector=true
   ) {module=InitialMeshOrientation001}
   &InitialMeshOrientation(
    Mesh Orientation Mode=::NewEnumerator7
    Rotation=&MaskVectorBySpawnGroup(Masked Vector=<0 0 0>, Passthrough Vector=<0 0 -0.25>)
    Use Rotation Vector=true
   ) {module=InitialMeshOrientation002}
   &InitialMeshOrientation(
    Mesh Orientation Mode=::NewEnumerator7
    Rotation=&MaskVectorBySpawnGroup(Masked Vector=<0 0 0>, Passthrough Vector=<0.25 0 0>)
    Use Rotation Vector=true
   ) {module=InitialMeshOrientation003}
  ]
  particleUpdateScript [
   &ParticleState(Kill Particles When Lifetime Has Elapsed=false, Let Infinitely Lived Particles Die When Emitter Deactivates=true)
   &SkeletalMeshLocation(
    Bone / Socket Index=$Particles.PartitionParticles.Partition
    MaskSpawnGroup=false
    Mesh Sampling Type=::NewEnumerator5
    Sampled Position Offset=&MaskVector(Mask Source=0, Mask Parameter=?, Mask=0, Masked Vector=<0 0 0>, Passthrough Vector=<0 0 0>)
    Skeletal Mesh=$System.Skeletal Mesh
    Socket Sampling Mode=::NewEnumerator5
   ) {module=SkeletalMeshLocation001}
   &ScaleMeshSize(Scale Factor=&MaskVector(Mask Source=0, Mask Parameter=?, Mask=0, Masked Vector=<1 1 1>, Passthrough Vector=<1 1 1.15>)) {module=ScaleMeshSize004}
   &ScaleColor(
    Scale Alpha=&Multiply_Float(
     A=&FloatFromCurve(
      FloatCurve=DataInterface() {propertyValues {Curve {keys [
          {arriveTangent=-0.3486, interpMode=RCIM_Cubic, leaveTangent=-0.3486, tangentMode=RCTM_User, time=0, value=0}
          {arriveTangent=-0.1219, interpMode=RCIM_Cubic, leaveTangent=-0.1219, tangentMode=RCTM_User, time=1, value=1}
         ]}}}
      Scale Curve=20
     )
     B=$System.FadeSlow
    )
    ScaleA=true
    ScaleRGB=true
   )
  ]
  rendererClasses [&NiagaraMeshRendererProperties]
  renderers [{rendererClass=&NiagaraMeshRendererProperties, rendererIndex=0}]
 }
 InnerGlowCard {
  emitterUpdateScript [
   &EmitterState()
   &SpawnBurst_Instantaneous(Spawn Count=2, Use Spawn Probability=false, UseLoopCountLimit=true)
   &SpawnBurst_Instantaneous(Spawn Count=4, Spawn Group=1, Use Spawn Probability=false, UseLoopCountLimit=true) {module=SpawnBurst_Instantaneous001}
  ]
  particleSpawnScript [
   &InitializeParticle(
    Color=&MaskLinearColor(Mask Source=0, Mask Parameter=?, Mask=0, Masked Linear Color=<rgba 5 1.2792 0 0.5>, Passthrough Linear Color=<rgba 0 2 5 0.5>)
    Sprite Size Mode=::NewEnumerator3
    Uniform Sprite Size=20
    UsePositionOffset=true
    Write Material Random=true
   )
   /SetVariables(Particles.Partition:NiagaraInt32=&SelectIntFromArray(
     Direct Array Index=&ModuloInt(A=&ReturnParticleIDAttributes(), B=6)
     Int Selection Array=DataInterface() {propertyValues {IntData [0,1,2,3,4,5]}}
    )) {isSetParameters=true}
   &CameraOffset(Camera Offset Amount=&MaskFloatBySpawnGroup(Masked Float=0, Passthrough Float=3))
  ]
  particleUpdateScript [
   &ParticleState(Kill Particles When Lifetime Has Elapsed=false, Let Infinitely Lived Particles Die When Emitter Deactivates=true)
   &ScaleSpriteSize(
    Scale Factor=/Niagara/DynamicInputs/Helpers/MaskVector2d(
     Mask Source:NiagaraInt32=0
     Mask Parameter:NiagaraInt32=?
     Mask:NiagaraInt32=0
     Masked Vector2D:Vector2f=<1 1>
     Passthrough Vector2D:Vector2f=<0.8 0.6>
    )
    Scale Sprite Size Mode=::NewEnumerator2
   )
   &SkeletalMeshLocation(
    Bone / Socket Index=$Particles.Initial.Partition
    MaskSpawnGroup=false
    Mesh Sampling Type=::NewEnumerator5
    Sampled Position Offset=&MaskVectorBySpawnGroup(Masked Vector=<0 0 0>, Passthrough Vector=<0 0 -4>)
    Skeletal Mesh=$System.Skeletal Mesh
    Socket Sampling Mode=::NewEnumerator5
   )
   &SpriteFacingAndAlignment(
    Sprite Alignment=&RotateVectorByQuaternion(
     Vector To Rotate=&MaskVector(Mask Source=0, Mask Parameter=?, Mask=0, Masked Vector=<1 0 0>, Passthrough Vector=<0 0 1>)
     Quaternion=$Particles.MeshOrientation
    )
    Sprite Facing=&MaskVector(
     Mask Source=0
     Mask Parameter=?
     Mask=0
     Masked Vector=<0 1 0>
     Passthrough Vector=&RotateVectorByQuaternion(Vector To Rotate=<1 0 0>, Quaternion=$Particles.SkeletalMeshLocation.SampledRotation)
    )
    WriteAlignment=true
    WriteFacing=true
   )
   &ScaleColor(
    Scale Alpha=&Multiply_Float(A=&Lerp_Float(A=0.125, B=0.125, Alpha=$System.Timeline.NormalizedTime), B=$System.FadeSlow)
    ScaleA=true
    ScaleRGB=false
   )
   &FN_Sprite_Attributes(
    Color=&MultiplyLinearColor(
     Linear Color=$Particles.Color
     Scale Factor=&MakeVector4FromVectorAndFloat("Vector (RGB)"=&VectorFromFloat(Value=30), "Float (Alpha)"=1.5)
    )
    Create Color Attribute=true
    Create Position Attribute=true
    Create SpriteSize Attribute=true
    Module Output:ENiagara_AttributeSamplingApplyOutput=::NewEnumerator1
    Position:NiagaraPosition=&AddVectorToPosition(
     Position=$Particles.Position
     Vector=&MaskVector(
      Mask Source=0
      Mask Parameter=?
      Mask=0
      Masked Vector=<0 1 0>
      Passthrough Vector=&RotateVectorByQuaternion(Vector To Rotate=<0 0 4>, Quaternion=$Particles.SkeletalMeshLocation.SampledRotation)
     )
    )
    SpriteSize=&Multiply_Vector2D_ByFloat(Float=1, Vector2D=$Particles.SpriteSize)
   )
  ]
  rendererClasses [&NiagaraSpriteRendererProperties]
  renderers [
   {rendererClass=&NiagaraSpriteRendererProperties}
   {rendererClass=&NiagaraSpriteRendererProperties, rendererIndex=1}
  ]
 }
 Minimal {
  bEnabled=false
  emitterUpdateScript [
   &EmitterState()
   &SpawnBurst_Instantaneous(Spawn Count=8, Use Spawn Probability=false, UseLoopCountLimit=true)
  ]
  particleSpawnScript [
   &InitializeParticle(Color Mode=::NewEnumerator0, Sprite Size Mode=::NewEnumerator3, Uniform Sprite Size=50, UsePositionOffset=false, Write Material Random=true)
   &PartitionParticles(Number Of Partitions=8)
  ]
  particleUpdateScript [
   &ParticleState(Kill Particles When Lifetime Has Elapsed=false, Loop Particles Lifetime=true)
   &SkeletalMeshLocation(
    Bone / Socket Index=$Particles.PartitionParticles.Partition
    MaskSpawnGroup=false
    Mesh Sampling Type=::NewEnumerator5
    Skeletal Mesh=$System.Skeletal Mesh
    Socket Sampling Mode=::NewEnumerator5
   )
  ]
  rendererClasses [&NiagaraSpriteRendererProperties]
  renderers [{rendererClass=&NiagaraSpriteRendererProperties}]
 }
 SideWingsinner {
  emitterSpawnScript [&FN_RibbonIDOffsetUtility()]
  emitterUpdateScript [
   &EmitterState()
   &SpawnBurst_Instantaneous(Spawn Count=10, Use Spawn Probability=false, UseLoopCountLimit=true)
  ]
  particleSpawnScript [
   &InitializeParticle(Sprite Size=<20 50>, Sprite Size Mode=::NewEnumerator4, UsePositionOffset=false, Write Material Random=true)
   &PartitionParticles() {module=PartitionParticles001}
   &FN_SetSpawnGroup(Inclusivity=true, Normalize Spawn Group=true, Particles Per Group=5, Pattern=::NewEnumerator0)
  ]
  particleUpdateScript [
   &ParticleState(Kill Particles When Lifetime Has Elapsed=false, Let Infinitely Lived Particles Die When Emitter Deactivates=true)
   &ScaleSpriteSize(
    Scale Factor=&MakeVector2D(X=1, Y=&Lerp_Float(A=0.5, B=1.5, Alpha=$System.Timeline.NormalizedTime))
    Scale Sprite Size Mode=::NewEnumerator2
   )
   &SkeletalMeshLocation(
    Bone / Socket Index=&SelectIntFromArray(Direct Array Index=$Particles.PartitionParticles001.Partition, Int Selection Array=DataInterface() {propertyValues {IntData [6,7]}})
    MaskSpawnGroup=false
    Mesh Sampling Type=::NewEnumerator5
    Sampled Position Offset=&MakeVector(
     X=0
     Y=&FN_FrontEndFloatModifier(Gameplay=20, Front End=10)
     Z=&Multiply_Float(
      A=$Particles.Initial.Normalized Spawn Group
      B=&FN_FrontEndFloatModifier(Gameplay=27, Front End=20)
     )
    )
    Skeletal Mesh=$System.Skeletal Mesh
    Socket Sampling Mode=::NewEnumerator5
   )
   &SpriteFacingAndAlignment(
    Sprite Alignment=&RotateVectorByQuaternion(Vector To Rotate=<0 -1 0>, Quaternion=$Particles.SkeletalMeshLocation.SampledRotation)
    Sprite Facing=&RotateVectorByQuaternion(Vector To Rotate=<0 0 1>, Quaternion=$Particles.SkeletalMeshLocation.SampledRotation)
    WriteAlignment=true
    WriteFacing=true
   )
   &ScaleColor(
    Scale Alpha=&Multiply_Float(A=&Lerp_Float(A=0.025, B=0.075, Alpha=$System.Timeline.NormalizedTime), B=$System.FadeSlow)
    ScaleA=true
    ScaleRGB=true
   ) {module=ScaleColor001}
   &DynamicMaterialParameters(
    Index 0 Param 1=1
    Index 0 Param 4=1
    Index 1 Param 1=&RandomRangeFloat(Minimum=0, Maximum=1)
    Index 1 Param 2=&RandomRangeFloat(Minimum=0, Maximum=1)
    Module.Index0Float0Write=true
    Module.Index0Float1Write=false
    Module.Index0Float2Write=false
    Module.Index0Float3Write=true
    Module.Index1Float0Write=true
    Module.Index1Float1Write=true
    Module.Index1Float2Write=false
    Module.Index1Float3Write=false
    Write Parameter Index 1=::NewEnumerator1
   )
  ]
  rendererClasses [&NiagaraSpriteRendererProperties]
  renderers [{rendererClass=&NiagaraSpriteRendererProperties}]
 }
 SpeedLine {
  bEnabled=false
  emitterUpdateScript [
   &EmitterState()
   &SpawnRate(
    SpawnRate=&MakeCustomFloatFromBool(Bool=$User.bIsFrontEndPreview, True Float=50, False Float=90)
    Use Spawn Probability=true
   ) {module=SpawnRate001}
  ]
  particleSpawnScript [
   &InitializeParticle(
    Color=&MakeLinearColorFromVectorAndFloat(
     "Vector (RGB)"=<1 1 1>
     "Float (Alpha)"=&Multiply_Float(
      A=&UniformRangedFloat(Minimum=0.1, Maximum=0.18)
      B=&MakeCustomFloatFromBool(Bool=$User.bIsFrontEndPreview, True Float=0.5, False Float=1)
     )
    )
    Lifetime Max=0.16
    Lifetime Min=0.1
    Lifetime Mode=::NewEnumerator1
    Sprite Size Max=<3 10>
    Sprite Size Min=<2 2>
    Sprite Size Mode=::NewEnumerator2
    UsePositionOffset=false
    Write Material Random=true
   )
   &TorusLocation(
    Align Torus With Axis=true
    Handle Radius=0
    Large Radius:NiagaraFloat=&RandomRangeFloat(Minimum=120, Maximum=160)
    MaskSpawnGroup=false
    Module.Override Randomness=false
    Offset=<0 -180 0>
    Torus Axis=<0 1 0>
    Torus Origin:NiagaraPosition=&Add_Vector(
     A=Linked() {linkedVariable {name=Particles.Position, type=NiagaraPosition}}
     B=&MakeVector(
      X=0
      Y=&MakeCustomFloatFromBool(Bool=$User.bIsFrontEndPreview, True Float=-250, False Float=0)
      Z=&FN_FrontEndFloatModifier(Gameplay=0, Front End=-50)
     )
    )
   )
   &AddVelocityFromPoint(
    Module.Use Falloff Distance=false
    Origin Offset=<0 -500 0>
    Velocity Strength=&Multiply_Float(A=600, B=&MakeCustomFloatFromBool(Bool=$User.bIsFrontEndPreview, True Float=0.25, False Float=1))
   )
   &AddVelocity(Velocity=&Multiply_Vector(
     A=&RandomRangeVector(Minimum=<0 300 0>, Maximum=<0 700 0>)
     B=&MakeVector(X=1, Y=&MakeCustomFloatFromBool(Bool=$User.bIsFrontEndPreview, True Float=1, False Float=1), Z=1)
    ))
  ]
  particleUpdateScript [
   &SolveForcesAndVelocity(Clamp Velocity=false, Limit Acceleration=false)
   &ParticleState()
   &ScaleSpriteSize(Scale Factor=&Vector2DFromCurve(
     Vector2Curve=DataInterface() {propertyValues {XCurve {keys [<at 0 0.5 RCIM_Cubic>,<at 0.5 1 RCIM_Cubic>]}, YCurve {keys [<at 0 0 RCIM_Cubic>,<at 0.5 1 RCIM_Cubic>]}}}
     ScaleCurve=&Vector2DFromFloat(Value=1)
    ))
   &ScaleSpriteSizeBySpeed(Max Scale Factor=<1 50>, Min Scale Factor=<0.01 0.05>, Velocity Threshold=1)
   &ScaleColor(
    Scale Alpha=&FloatFromCurve(
     FloatCurve=DataInterface() {propertyValues {Curve {keys [
         <at 0 0 RCIM_Cubic>
         <at 0.3759 1.0207>
         {arriveTangent=-2.221, interpMode=RCIM_Cubic, leaveTangent=-2.221, tangentMode=RCTM_User, time=1, value=0}
        ]}}}
     Scale Curve=&MakeCustomFloatFromBool(
      Bool=$User.bIsFrontEndPreview
      True Float=&MakeCustomFloatFromBool(Bool=$User.bIsFullyDeployed, True Float=1, False Float=0)
      False Float=&Lerp_Float(A=0, B=0.66, Alpha=$User.ForwardDot)
     )
    )
    Scale RGB=&VectorFromFloat(Value=1)
    ScaleA=true
    ScaleRGB=true
   )
  ]
  rendererClasses [&NiagaraSpriteRendererProperties]
  renderers [{rendererClass=&NiagaraSpriteRendererProperties}]
 }
 Speedlines {
  emitterUpdateScript [
   &EmitterState()
   &SpawnRate(
    SpawnRate=&MakeCustomFloatFromBool(Bool=$User.bIsFrontEndPreview, True Float=50, False Float=90)
    Use Spawn Probability=true
   ) {module=SpawnRate001}
  ]
  particleSpawnScript [
   &InitializeParticle(
    Color=&MakeLinearColorFromVectorAndFloat(
     "Vector (RGB)"=<1 1 1>
     "Float (Alpha)"=&Multiply_Float(
      A=&UniformRangedFloat(Minimum=1, Maximum=1)
      B=&MakeCustomFloatFromBool(
       Bool=$User.bIsFrontEndPreview
       True Float=&FloatFromCurve(
        FloatCurve=DataInterface() {propertyValues {Curve {keys [{tangentMode=RCTM_User, time=0, value=0}, {tangentMode=RCTM_User, time=2, value=1}]}}}
        CurveIndex=$Emitter.Age
        Scale Curve=0.4
       )
       False Float=1
      )
     )
    )
    Lifetime Max=0.3
    Lifetime Min=0.2
    Lifetime Mode=::NewEnumerator1
    Position Offset=&FN_FrontEndVector3Modifier(Gameplay=<0 -50 0>, Front End=<0 50 0>)
    Sprite Size Max=<3 10>
    Sprite Size Min=<2 2>
    Sprite Size Mode=::NewEnumerator2
    UsePositionOffset=true
    Write Material Random=true
   )
   &AddVelocity(
    Module.Use Falloff Distance=false
    Override Seed=false
    Velocity Mode=::NewEnumerator1
    Velocity Speed=&FN_FrontEndFloatModifier(Gameplay=-2500, Front End=2500)
   ) {module=AddVelocity001}
   &TorusLocation(
    Align Torus With Axis=true
    Coordinate Space:ENiagaraCoordinateSpace=::Simulation
    Handle Radius=0
    Large Radius:NiagaraFloat=&FN_FrontEndFloatModifier(
     Gameplay=&RandomRangeFloat(Minimum=350, Maximum=600)
     Front End=&RandomRangeFloat(Minimum=250, Maximum=300)
    )
    MaskSpawnGroup=false
    Module.Override Randomness=false
    Offset=<0 33 0>
    Torus Axis=<0 1 0>
    Torus Origin:NiagaraPosition=&Add_Vector(
     A=Linked() {linkedVariable {name=Particles.Position, type=NiagaraPosition}}
     B=&MakeVector(
      X=0
      Y=&MakeCustomFloatFromBool(Bool=$User.bIsFrontEndPreview, True Float=-700, False Float=-1000)
      Z=&FN_FrontEndFloatModifier(Gameplay=0, Front End=-50)
     )
    )
   )
   &AddVelocity(
    Velocity=&Add_Vector(
     A=&RandomRangeVector(Minimum=<0 300 0>, Maximum=<0 700 1>)
     B=&Add_Vector(
      A=&Multiply_VectorByFloat(Vector=<-555 0 0>, Float=$User.RightDot)
      B=&Multiply_VectorByFloat(Vector=<0 0 222>, Float=$User.ForwardDot)
     )
    )
    Velocity Speed Scale=&FN_FrontEndFloatModifier(Gameplay=1, Front End=0.3)
   )
   &AddVelocity(Velocity=&Multiply_Vector(
     A=&RandomRangeVector(Minimum=<0 300 0>, Maximum=<0 700 0>)
     B=&MakeVector(X=1, Y=&MakeCustomFloatFromBool(Bool=$User.bIsFrontEndPreview, True Float=-1, False Float=1), Z=1)
    )) {enabled=false, module=AddVelocity002}
   &AddVelocityFromPoint(
    Module.Use Falloff Distance=false
    Origin Offset=<0 -500 0>
    Velocity Strength=&Multiply_Float(A=600, B=&MakeCustomFloatFromBool(Bool=$User.bIsFrontEndPreview, True Float=0.25, False Float=1))
   ) {enabled=false}
  ]
  particleUpdateScript [
   &SolveForcesAndVelocity(Clamp Velocity=false, Limit Acceleration=false)
   &ParticleState()
   &ScaleSpriteSize(Scale Factor=&Vector2DFromCurve(
     Vector2Curve=DataInterface() {propertyValues {XCurve {keys [<at 0 0.5 RCIM_Cubic>,<at 0.5 1 RCIM_Cubic>]}, YCurve {keys [<at 0 0 RCIM_Cubic>,<at 0.5 1 RCIM_Cubic>]}}}
     ScaleCurve=&Vector2DFromFloat(Value=1)
    ))
   &ScaleSpriteSizeBySpeed(Max Scale Factor=<1 50>, Min Scale Factor=<0.01 0.05>, Velocity Threshold=1)
   &ScaleColor(
    Scale Alpha=&FloatFromCurve(
     FloatCurve=DataInterface() {propertyValues {Curve {keys [
         <at 0 0 RCIM_Cubic>
         <at 0.3759 1.0207>
         {arriveTangent=-2.221, interpMode=RCIM_Cubic, leaveTangent=-2.221, tangentMode=RCTM_User, time=1, value=0}
        ]}}}
     Scale Curve=&MakeCustomFloatFromBool(
      Bool=$User.bIsFrontEndPreview
      True Float=&MakeCustomFloatFromBool(Bool=$User.bIsFullyDeployed, True Float=0.5, False Float=0)
      False Float=&Lerp_Float(A=0, B=0.5, Alpha=&FN_FrontEndFloatModifier(Gameplay=$System.Timeline.Time, Front End=0.5))
     )
    )
    Scale RGB=&VectorFromFloat(Value=&FN_FrontEndFloatModifier(Gameplay=0.5, Front End=0.33))
    ScaleA=true
    ScaleRGB=true
   )
  ]
  rendererClasses [&NiagaraSpriteRendererProperties]
  renderers [{rendererClass=&NiagaraSpriteRendererProperties}]
 }
 Thrusters {
  emitterUpdateScript [
   &EmitterState()
   &SpawnBurst_Instantaneous(Spawn Count=2, Use Spawn Probability=false, UseLoopCountLimit=true) {module=SpawnBurst_Instantaneous001}
   &SpawnBurst_Instantaneous(Spawn Count=4, Spawn Group=1, Use Spawn Probability=false, UseLoopCountLimit=true) {module=SpawnBurst_Instantaneous002}
   /SetVariables(
    Emitter.DarkColor:LinearColor=&LerpLinearColors(StartColor=<rgba 0 0.2 0.3478 50>, EndColor=<rgba 0 0.4 0.5 1>, LerpFactor=$System.Timeline.NormalizedTime)
    Emitter.LightColor:LinearColor=&LerpLinearColors(StartColor=<rgba 0 0.8 1 100>, EndColor=<rgba 0 2.4 3 100>, LerpFactor=$System.Timeline.NormalizedTime)
   ) {isSetParameters=true}
   /SetVariables(
    Emitter.DarkColor_Orange:LinearColor=&Multiply_LinearColorByFloat(
     Linear Color=&LerpLinearColors(StartColor=<rgba 1 0.2899 0 50>, EndColor=<rgba 1.25 0.3497 0 50>, LerpFactor=$System.Timeline.NormalizedTime)
     Float=&FloatFromCurve(
      FloatCurve=DataInterface() {propertyValues {Curve {
         defaultValue=1
         keys [
          {arriveTangent=-4.5706, interpMode=RCIM_Cubic, leaveTangent=-4.5706, tangentMode=RCTM_User, time=0, value=0}
          {arriveTangent=-4.0131, interpMode=RCIM_Cubic, leaveTangent=-4.0131, tangentMode=RCTM_User, time=0.5, value=8}
          {arriveTangent=-4.5706, interpMode=RCIM_Cubic, leaveTangent=-4.5706, tangentMode=RCTM_User, time=0.75, value=1}
         ]
        }}}
      CurveIndex=$Engine.System.Age
     )
    )
    Emitter.LightColor_Orange:LinearColor=&Multiply_LinearColorByFloat(
     Linear Color=&LerpLinearColors(StartColor=<rgba 2 0.3016 0 100>, EndColor=<rgba 2 0.3739 0 100>, LerpFactor=$System.Timeline.NormalizedTime)
     Float=&FloatFromCurve(
      FloatCurve=DataInterface() {propertyValues {Curve {
         defaultValue=1
         keys [
          {arriveTangent=-4.5706, interpMode=RCIM_Cubic, leaveTangent=-4.5706, tangentMode=RCTM_User, time=0, value=0}
          {arriveTangent=-4.0131, interpMode=RCIM_Cubic, leaveTangent=-4.0131, tangentMode=RCTM_User, time=0.5, value=8}
          {arriveTangent=-4.5706, interpMode=RCIM_Cubic, leaveTangent=-4.5706, tangentMode=RCTM_User, time=0.75, value=1}
         ]
        }}}
      CurveIndex=$Engine.System.Age
     )
    )
   ) {isSetParameters=true}
  ]
  particleSpawnScript [
   &PartitionParticles(Number Of Partitions=6)
   &InitializeParticle(
    Color=&MaskLinearColor(Mask Source=0, Mask Parameter=?, Mask=0, Masked Linear Color=<rgba 3 1.2389 0 100>, Passthrough Linear Color=<rgba 0 2.4 3 100>)
    Mesh Scale=&Multiply_VectorByFloat(Vector=<1.25 0.07 0.07>, Float=&FN_FrontEndFloatModifier(Gameplay=1, Front End=0.7))
    Mesh Scale Mode=::NewEnumerator4
    Sprite Size Mode=::NewEnumerator3
    Uniform Sprite Size=25
    UsePositionOffset=false
    Write Material Random=true
   )
   /SetVariables(Particles.Dynam_Param2:Vector4f=<1 5 5 5>) {isSetParameters=true}
   &InitialMeshOrientation(
    Mesh Orientation Mode=::NewEnumerator7
    Rotation=&MaskVectorBySpawnGroup(Masked Vector=<0 0 -0.25>, Passthrough Vector=<0 0.5 0>)
    Use Rotation Vector=true
   )
   &DynamicMaterialParameters(
    Index 0 Param 1=2
    Index 0 Param 2=5
    Index 0 Param 3=5
    Index 0 Param 4=5
    Module.Index0Param0Write=true
    Module.Index0Param1Write=true
    Module.Index0Param2Write=true
    Module.Index0Param3Write=true
   )
   /SetVariables(Particles.VisibilityTag=&MaskIntBySpawnGroup(Masked Int=0, Passthrough Int=1, Spawn Group Mask=0)) {isSetParameters=true}
  ]
  particleUpdateScript [
   &ParticleState(Kill Particles When Lifetime Has Elapsed=false, Let Infinitely Lived Particles Die When Emitter Deactivates=true)
   &InterpolateOverTime(Target Float Value:NiagaraFloat=&ClampFloat(Float=$User.ForwardDot))
   &SkeletalMeshLocation(
    Bone / Socket Index=$Particles.PartitionParticles.Partition
    MaskSpawnGroup=false
    Mesh Sampling Type=::NewEnumerator5
    Sampled Position Offset=&MaskVector(Mask Source=0, Mask Parameter=?, Mask=0, Masked Vector=<0 5 0>, Passthrough Vector=<0 0 0>)
    Skeletal Mesh=$System.Skeletal Mesh
    Socket Sampling Mode=::NewEnumerator5
   )
   /SetVariables(Particles.InnerFlame_Color:LinearColor=&MakeLinearColorFromVectorAndFloat(
     "Vector (RGB)"=&VectorFromFloat(Value=&Lerp_Float(A=1, B=1, Alpha=$System.Timeline.Time))
     "Float (Alpha)"=1
    )) {isSetParameters=true}
   &ScaleMeshSize(Scale Factor=&Lerp_Vector(A=<0.2 1 1>, B=<0.45 1 1>, Alpha=$Particles.InterpolateOverTime.InterpolateOverTime_Float.Moving Average)) {module=ScaleMeshSize001}
   &ScaleMeshSize(Scale Factor=&MakeVector(
     X=&FloatFromCurve(
      FloatCurve=DataInterface() {propertyValues {Curve {keys [
          {arriveTangent=0.2726, interpMode=RCIM_Cubic, leaveTangent=0.2726, tangentMode=RCTM_User, time=0, value=0}
          {arriveTangent=11.813, interpMode=RCIM_Cubic, leaveTangent=11.813, tangentMode=RCTM_User, time=0.3, value=1}
          {arriveTangent=0.0777, interpMode=RCIM_Cubic, leaveTangent=0.0777, tangentMode=RCTM_User, time=0.35, value=2}
          {arriveTangent=-0.0405, interpMode=RCIM_Cubic, leaveTangent=-0.0405, tangentMode=RCTM_User, time=1, value=1}
         ]}}}
      CurveIndex=$Engine.System.Age
     )
     Y=&FloatFromCurve(
      FloatCurve=DataInterface() {propertyValues {Curve {keys [
          {arriveTangent=0.667, interpMode=RCIM_Cubic, leaveTangent=0.667, tangentMode=RCTM_User, time=0, value=0.75}
          {arriveTangent=0.25, interpMode=RCIM_Cubic, leaveTangent=0.25, tangentMode=RCTM_User, time=0.35, value=2}
          {arriveTangent=-0.0405, interpMode=RCIM_Cubic, leaveTangent=-0.0405, tangentMode=RCTM_User, time=1, value=1}
         ]}}}
      CurveIndex=$Engine.System.Age
     )
     Z=&FloatFromCurve(
      FloatCurve=DataInterface() {propertyValues {Curve {keys [
          {arriveTangent=0.667, interpMode=RCIM_Cubic, leaveTangent=0.667, tangentMode=RCTM_User, time=0, value=0.75}
          {arriveTangent=0.25, interpMode=RCIM_Cubic, leaveTangent=0.25, time=0.35, value=2}
          {arriveTangent=-0.0405, interpMode=RCIM_Cubic, leaveTangent=-0.0405, tangentMode=RCTM_User, time=1, value=1}
         ]}}}
      CurveIndex=$Engine.System.Age
     )
    )) {module=ScaleMeshSize002}
   &ScaleMeshSize(Scale Factor=&MaskVector(Mask Source=0, Mask Parameter=?, Mask=0, Masked Vector=<1 1 1>, Passthrough Vector=<0.35 0.8 0.7>)) {module=ScaleMeshSize003}
  ]
  rendererClasses [&NiagaraMeshRendererProperties]
  renderers [
   {rendererClass=&NiagaraMeshRendererProperties, rendererIndex=0}
   {rendererClass=&NiagaraMeshRendererProperties, rendererIndex=1}
  ]
 }
 Thrusters001 {
  emitterUpdateScript [
   &EmitterState()
   &SpawnBurst_Instantaneous(Spawn Count=2, Use Spawn Probability=false, UseLoopCountLimit=true)
   &SpawnBurst_Instantaneous(Spawn Count=4, Spawn Group=1, Use Spawn Probability=false, UseLoopCountLimit=true) {module=SpawnBurst_Instantaneous002}
   /SetVariables(
    Emitter.DarkColor:LinearColor=&LerpLinearColors(StartColor=<rgba 0 0.2 0.3478 50>, EndColor=<rgba 0 0.4 0.5 1>, LerpFactor=$System.Timeline.NormalizedTime)
    Emitter.LightColor:LinearColor=&LerpLinearColors(StartColor=<rgba 0 0.8 1 100>, EndColor=<rgba 0 2.4 3 100>, LerpFactor=$System.Timeline.NormalizedTime)
   ) {isSetParameters=true}
   /SetVariables(
    Emitter.DarkColor_Orange:LinearColor=&Multiply_LinearColorByFloat(
     Linear Color=&LerpLinearColors(StartColor=<rgba 1 0.2899 0 50>, EndColor=<rgba 1.25 0.3497 0 50>, LerpFactor=$System.Timeline.NormalizedTime)
     Float=&FloatFromCurve(
      FloatCurve=DataInterface() {propertyValues {Curve {
         defaultValue=1
         keys [
          {arriveTangent=-4.5706, interpMode=RCIM_Cubic, leaveTangent=-4.5706, tangentMode=RCTM_User, time=0, value=0}
          {arriveTangent=-4.0131, interpMode=RCIM_Cubic, leaveTangent=-4.0131, tangentMode=RCTM_User, time=0.5, value=3}
          {arriveTangent=-4.5706, interpMode=RCIM_Cubic, leaveTangent=-4.5706, tangentMode=RCTM_User, time=0.75, value=1}
         ]
        }}}
      CurveIndex=$Engine.System.Age
     )
    )
    Emitter.LightColor_Orange:LinearColor=&Multiply_LinearColorByFloat(
     Linear Color=&LerpLinearColors(StartColor=<rgba 2 0.3016 0 100>, EndColor=<rgba 2 0.3739 0 100>, LerpFactor=$System.Timeline.NormalizedTime)
     Float=&FloatFromCurve(
      FloatCurve=DataInterface() {propertyValues {Curve {
         defaultValue=1
         keys [
          {arriveTangent=-4.5706, interpMode=RCIM_Cubic, leaveTangent=-4.5706, tangentMode=RCTM_User, time=0, value=0}
          {arriveTangent=-4.0131, interpMode=RCIM_Cubic, leaveTangent=-4.0131, tangentMode=RCTM_User, time=0.5, value=3}
          {arriveTangent=-4.5706, interpMode=RCIM_Cubic, leaveTangent=-4.5706, tangentMode=RCTM_User, time=0.75, value=1}
         ]
        }}}
      CurveIndex=$Engine.System.Age
     )
    )
   ) {isSetParameters=true}
  ]
  particleSpawnScript [
   &PartitionParticles(Number Of Partitions=6, Write Partition to Ribbon ID=::NewEnumerator0)
   &InitializeParticle(
    Color=&MaskLinearColor(Mask Source=0, Mask Parameter=?, Mask=0, Masked Linear Color=<rgba 1 1 1 100>, Passthrough Linear Color=<rgba 0 1.8 3 100>)
    Mesh Scale=&Multiply_Vector(
     A=&MaskVectorBySpawnGroup(Masked Vector=<1.75 0.14 0.14>, Passthrough Vector=<0.8 0.12 0.114>)
     B=&FN_FrontEndVector3Modifier(Gameplay=<1 1 1>, Front End=<1 0.7 0.7>)
    )
    Mesh Scale Mode=::NewEnumerator4
    Sprite Size Mode=::NewEnumerator3
    Uniform Sprite Size=25
    UsePositionOffset=false
    Write Material Random=true
   )
   &InitialMeshOrientation(Mesh Orientation Mode=::NewEnumerator7, Rotation=<0 0 -0.25>, Use Rotation Vector=true)
   &InitialMeshOrientation(
    Mesh Orientation Mode=::NewEnumerator7
    Rotation=&MaskVectorBySpawnGroup(Masked Vector=<0 0 0>, Passthrough Vector=<0 0 -0.25>)
    Use Rotation Vector=true
   ) {module=InitialMeshOrientation001}
   &DynamicMaterialParameters(
    Index 0 Param 1=1
    Index 0 Param 2=-2
    Index 0 Param 3=&SelectFloatFromArray(Direct Array Index=$Particles.PartitionParticles.Partition, Float Selection Array=DataInterface() {propertyValues {FloatData [0,0.5]}})
    Index 0 Param 4=&SelectFloatFromArray(Direct Array Index=$Particles.PartitionParticles.Partition, Float Selection Array=DataInterface() {propertyValues {FloatData [0,0.5]}})
    Module.Index0Param0Write=true
    Module.Index0Param1Write=true
    Module.Index0Param2Write=true
    Module.Index0Param3Write=true
   )
   /SetVariables(Particles.VisibilityTag=&MaskIntBySpawnGroup(Masked Int=0, Passthrough Int=1, Spawn Group Mask=0)) {isSetParameters=true}
  ]
  particleUpdateScript [
   &ParticleState(Kill Particles When Lifetime Has Elapsed=false, Let Infinitely Lived Particles Die When Emitter Deactivates=true)
   &InterpolateOverTime(Target Float Value:NiagaraFloat=&ClampFloat(Float=$User.ForwardDot))
   &SkeletalMeshLocation(
    Bone / Socket Index=$Particles.PartitionParticles.Partition
    MaskSpawnGroup=false
    Mesh Sampling Type=::NewEnumerator5
    Sampled Position Offset=&MaskVector(Mask Source=0, Mask Parameter=?, Mask=0, Masked Vector=<0 0 0>, Passthrough Vector=<0 0 0>)
    Skeletal Mesh=$System.Skeletal Mesh
    Socket Sampling Mode=::NewEnumerator5
   ) {module=SkeletalMeshLocation001}
   &ScaleMeshSize(Scale Factor=&Lerp_Vector(A=<0.25 1 1>, B=<0.5 1 1>, Alpha=$Particles.InterpolateOverTime.InterpolateOverTime_Float.Moving Average)) {module=ScaleMeshSize001}
   &ScaleMeshSize(Scale Factor=&MakeVector(
     X=&FloatFromCurve(
      FloatCurve=DataInterface() {propertyValues {Curve {keys [
          {arriveTangent=0.2726, interpMode=RCIM_Cubic, leaveTangent=0.2726, tangentMode=RCTM_User, time=0, value=0}
          {arriveTangent=12.8569, interpMode=RCIM_Cubic, leaveTangent=12.8569, tangentMode=RCTM_User, time=0.3, value=1}
          {arriveTangent=0.0777, interpMode=RCIM_Cubic, leaveTangent=0.0777, tangentMode=RCTM_User, time=0.35, value=3}
          {arriveTangent=-0.0405, interpMode=RCIM_Cubic, leaveTangent=-0.0405, tangentMode=RCTM_User, time=1, value=1}
         ]}}}
      CurveIndex=$Engine.System.Age
     )
     Y=&FloatFromCurve(
      FloatCurve=DataInterface() {propertyValues {Curve {keys [
          {arriveTangent=0.667, interpMode=RCIM_Cubic, leaveTangent=0.667, tangentMode=RCTM_User, time=0, value=0.75}
          {arriveTangent=0.25, interpMode=RCIM_Cubic, leaveTangent=0.25, time=0.35, value=1.25}
          {arriveTangent=-0.0405, interpMode=RCIM_Cubic, leaveTangent=-0.0405, tangentMode=RCTM_User, time=1, value=1}
         ]}}}
      CurveIndex=$Engine.System.Age
     )
     Z=&FloatFromCurve(
      FloatCurve=DataInterface() {propertyValues {Curve {keys [
          {arriveTangent=0.667, interpMode=RCIM_Cubic, leaveTangent=0.667, tangentMode=RCTM_User, time=0, value=0.75}
          {arriveTangent=0.25, interpMode=RCIM_Cubic, leaveTangent=0.25, time=0.35, value=1.25}
          {arriveTangent=-0.0405, interpMode=RCIM_Cubic, leaveTangent=-0.0405, tangentMode=RCTM_User, time=1, value=1}
         ]}}}
      CurveIndex=$Engine.System.Age
     )
    )) {module=ScaleMeshSize003}
   /SetVariables(Particles.Scale_Alt:Vector3f=&Multiply_Vector(
     A=<0.5 0.114 0.114>
     B=&MakeVector(
      X=&FloatFromCurve(
       FloatCurve=DataInterface() {propertyValues {Curve {keys [
           {arriveTangent=0.2726, interpMode=RCIM_Cubic, leaveTangent=0.2726, tangentMode=RCTM_User, time=0, value=0}
           {arriveTangent=12.8569, interpMode=RCIM_Cubic, leaveTangent=12.8569, tangentMode=RCTM_User, time=0.3, value=1}
           {arriveTangent=0.0777, interpMode=RCIM_Cubic, leaveTangent=0.0777, tangentMode=RCTM_User, time=0.35, value=3}
           {arriveTangent=-0.0405, interpMode=RCIM_Cubic, leaveTangent=-0.0405, tangentMode=RCTM_User, time=1, value=1}
          ]}}}
       CurveIndex=$Engine.System.Age
      )
      Y=&FloatFromCurve(
       FloatCurve=DataInterface() {propertyValues {Curve {keys [
           {arriveTangent=0.667, interpMode=RCIM_Cubic, leaveTangent=0.667, tangentMode=RCTM_User, time=0, value=0.75}
           {arriveTangent=0.25, interpMode=RCIM_Cubic, leaveTangent=0.25, time=0.35, value=1.25}
           {arriveTangent=-0.0405, interpMode=RCIM_Cubic, leaveTangent=-0.0405, tangentMode=RCTM_User, time=1, value=1}
          ]}}}
       CurveIndex=$Engine.System.Age
      )
      Z=&FloatFromCurve(
       FloatCurve=DataInterface() {propertyValues {Curve {keys [
           {arriveTangent=0.667, interpMode=RCIM_Cubic, leaveTangent=0.667, tangentMode=RCTM_User, time=0, value=0.75}
           {arriveTangent=0.25, interpMode=RCIM_Cubic, leaveTangent=0.25, time=0.35, value=1.25}
           {arriveTangent=-0.0405, interpMode=RCIM_Cubic, leaveTangent=-0.0405, tangentMode=RCTM_User, time=1, value=1}
          ]}}}
       CurveIndex=$Engine.System.Age
      )
     )
    )) {isSetParameters=true}
   &ScaleMeshSize(Scale Factor=&MaskVector(Mask Source=0, Mask Parameter=?, Mask=0, Masked Vector=<1 1 1>, Passthrough Vector=<0.5 0.65 0.75>)) {enabled=false, module=ScaleMeshSize004}
   &ScaleColor(ScaleA=true, ScaleRGB=true) {module=ScaleColor003}
   &FN_Sprite_Attributes(
    Create Position Attribute=true
    Module Output:ENiagara_AttributeSamplingApplyOutput=::NewEnumerator1
    Position:NiagaraPosition=&AddVectorToPosition(
     Position=$Particles.Position
     Vector=&RotateVectorByQuaternion(
      Vector To Rotate=&FN_FrontEndVector3Modifier(Gameplay=<-7 -17 0>, Front End=<-10 -9 0>)
      Quaternion=$Particles.SkeletalMeshLocation001.SampledRotation
     )
    )
   )
  ]
  rendererClasses [&NiagaraMeshRendererProperties]
  renderers [
   {rendererClass=&NiagaraMeshRendererProperties, rendererIndex=0}
   {rendererClass=&NiagaraMeshRendererProperties, rendererIndex=1}
   {rendererClass=&NiagaraMeshRendererProperties, rendererIndex=2}
  ]
 }
}
summary {
 systemName=NS_Glider_PaverYoung_Trail
 emitters [
  {emitterName=SideWingsinner, rendererClasses [&NiagaraSpriteRendererProperties]}
  {emitterName=Thrusters, rendererClasses [&NiagaraMeshRendererProperties]}
  {emitterName=Glows, rendererClasses [&NiagaraSpriteRendererProperties, &NiagaraLightRendererProperties]}
  {emitterName=Thrusters001, rendererClasses [&NiagaraMeshRendererProperties]}
  {emitterName=InnerFlameMesh, rendererClasses [&NiagaraMeshRendererProperties]}
  {emitterName=InnerGlowCard, rendererClasses [&NiagaraSpriteRendererProperties]}
  {emitterName=Speedlines, rendererClasses [&NiagaraSpriteRendererProperties]}
  {bEnabled=false, emitterName=Minimal, rendererClasses [&NiagaraSpriteRendererProperties]}
  {bEnabled=false, emitterName=SpeedLine, rendererClasses [&NiagaraSpriteRendererProperties]}
 ]
 userVariables [
  User.bIsFrontEnd {type=NiagaraBool}
  User.bIsFrontEndPreview {type=NiagaraBool}
  User.bIsFullyDeployed {type=NiagaraBool, defaultValue=true}
  User.ForwardDot {type=NiagaraFloat, defaultValue=1, description=""}
  User.RibbonIDOffset {type=NiagaraInt32}
  User.RightDot {type=NiagaraFloat, defaultValue=1, description=""}
 ]
}
